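; Green Glop foraged-item pickup with gravity, launched as a projectile and expiring shortly before the next pulse.
; Hits switch off the trail and play landing effects, with a distinct first-bounce sound; stopping simulation destroys the projectile component.
; A server timer, derived from the pulse interval, triggers a randomly staggered kill that sets replicated Hide, plays despawn FX and destroys the actor.
; Failed consumption stops the search montage and shows failure feedback for the matching local player.
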(event Collision|EventHit
  (params (MyComp :type Primitive%20Component%20Object%20Reference) (Other :type Actor%20Object%20Reference) (OtherComp :type Primitive%20Component%20Object%20Reference) (bSelfMoved :type Boolean) (HitLocation :type Vector :default "0, 0, 0") (HitNormal :type Vector :default "0, 0, 0") (NormalImpulse :type Vector :default "0, 0, 0") (Hit :type Hit%20Result%20Structure%20%28by%20ref%29))
  ;[1 1 1] On Hit turn off Trail
  (|CallParentFunction MyComp Other OtherComp bSelfMoved HitLocation HitNormal NormalImpulse Hit)
  (Components|Activation|Deactivate (Variables|Default|GetTrail)))

(event EventStopSim
  (params (Hit :type Hit%20Result%20Structure))
  ;[1 1 1] Remove Projectile movement component on stop sim
  (|CallParentFunction Hit)
  (Components|DestroyComponent (Variables|Default|GetProjectileMovement)))

(event EventBeginPlay
  ;[1 1 1] After a delay which is a little shorter than the Green Glop pulse duration, start hiding and killing the pickups.
  (|CallParentFunction)
  (if (Networking|IsServer)
    (Utilities|Time|SetTimerbyEvent (delegate Custom|Kill) (- (ScalableFloat|GetValueatLevel (Variables|Default|GetRow_PulseInterval)) (Variables|Default|GetPulseIntervalReduction)))))

(event Custom|Kill
  ;[1 1 1] After a delay which is a little shorter than the Green Glop pulse duration, start hiding and killing the pickups.
  (Utilities|FlowControl|Delay (Math|Random|RandomFloatinRange 0.0 2.0))
  (Variables|Default|SetHide true)
  (CallFunction|PlayFX)
  (Actor|DestroyActor))

(event Collision|EventActorBeginOverlap
  (params (OtherActor :type Actor%20Object%20Reference))
  (|CallParentFunction OtherActor) ; This node is disabled and will not be called.\nDrag off pins to build functionality.
)

(event Custom|PlayFX
  ;[1 1 1] Play FX when hiding
  (Effects|SpawnEmitteratLocation (Variables|Default|GetDespawnFX) (Transformation|GetWorldLocation (Variables|Default|GetTrail)) (Transformation|GetActorRotation)))

(event EventTick
  (params (DeltaSeconds :type Float%20%28single-precision%29))
  (|CallParentFunction DeltaSeconds) ; This node is disabled and will not be called.\nDrag off pins to build functionality.
)

(event EventOnAllowedToConsumeFailed
  (params (InteractingPawn :type Fort%20Pawn%20Object%20Reference))
  (Utilities|IsValid InteractingPawn
    (:Is%20Valid
      (bind _asfort_player_controller (Utilities|Casting|CastToFortPlayerController (Pawn|GetController InteractingPawn)))
      (CallFunction|HandleConsumeFailedVFX (Default|GetGameAccountId _asfort_player_controller)))))

(event Custom|HandleConsumeFailedVFX
  (params (InteractAccountId :type Unique%20Net%20Id%20Structure))
  (Networking|CanExecuteCosmeticEvents
    (:True
      (bind _array_element (Utilities|Array|ForEachLoopwithBreak (Fortnite|GetLocalFortPlayerControllers true))
        (:LoopBody
          (if (Online|UniqueNetIdRepl|Equal(UniqueIdRepl) (Default|GetGameAccountId _array_element) InteractAccountId)
            (Animation|StopAnimMontage (Game|Player|GetPlayerPawn _array_element) "/Game/Animation/Game/MainPlayer/Montages/General/Search.Search")
            (FailedAction|UnabletoPerformAction (Game|Player|GetPlayerPawn _array_element) true (Variables|Default|GetFailedReason_0))))))))

(event EventPlayHitFX
  (bind (_returnvalue_roll _returnvalue_pitch _returnvalue_yaw) (Math|Rotator|BreakRotator (Transformation|GetActorRotation)))
  (bind _hitgroundfx (Variables|Gravity|GetHitgroundFX))
  ;[1 1 1] Play Sounds and FX if valid
  (Utilities|IsValid _hitgroundfx
    (:Is%20Valid
      (Effects|SpawnEmitteratLocation _hitgroundfx (Transformation|GetActorLocation) (Math|Rotator|MakeRotator _returnvalue_roll 9e+01 _returnvalue_yaw) "1.000000,1.000000,1.000000" true "None" true))
    (:Is%20Not%20Valid
      _))
  (Audio|PlaySoundatLocation (select (Variables|Default|GetHasPlayedBounceSound) (Variables|Gravity|GetHitgroundSound) (Variables|Default|GetHitGroundSoundFirst)) (Transformation|GetActorLocation) "0, 0, 0" 1.0 1.0 0.0 0 0 self)
  (Variables|Default|SetHasPlayedBounceSound true))

(fn OnInteractAllowedToConsume
  (params (InteractingPawn :type Fort%20Pawn%20Object%20Reference) (InteractionBeingAttempted :type EInteractionBeingAttempted%20Enum :default "FirstInteraction"))
  (Utilities|IsValid InteractingPawn
    (:Is%20Valid
      (return (|CallParentFunction InteractingPawn "FirstInteraction" "IT_Simple")))
    (:Is%20Not%20Valid
      (return false))))

(fn OnRep_Hide
  (if (Variables|Default|GetHide)
    (Rendering|SetActorHiddenInGame true)))

(fn ConstructionScript
  (|CallParentFunction))
