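; Gameplay ability for the Green Glop foraged item that heals its owner, filling health first and sending any remainder to shield.
; On activation it fires a gameplay cue, reads the total from a scalable float, and splits it by missing health; computed missing shield goes unused.
; A short delay works around shield failing in a same-frame GA>GE grant chain.
; It then applies one effect spec with SetByCaller health and shield magnitudes and ends.
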(event Ability|EventActivateAbility
  (bind _returnvalue (Ability|GetAvatarActorfromActorInfo))
  ;[1 1 1] Apply Effect on tick
  (Ability|ExecuteGameplayCueOnOwner (Variables|Default|GetGC_Activate))
  (Utilities|IsValid _returnvalue
    (:Is%20Valid
      (bind _asfort_player_pawn (Utilities|Casting|CastToFortPlayerPawn _returnvalue))
      (Variables|Default|SetPlayerPawn _asfort_player_pawn)
      (Variables|Default|SetTotalHealAmount (ScalableFloat|GetValueatLevel (Variables|Default|GetRow_HealAmount)))
      (CallFunction|CalcMissingHealthShield)
      (Utilities|FlowControl|Delay 0.01) ; This is for some scope blocking that is stopping the shield from getting applied if it is in the same frame. Only happens because we have a GA>GE>GA>GE grant chain
      ;[1 1 1] Apply Heal effects
      (bind _returnvalue_2 (Ability|GameplayEffect|AssignTagSetbyCallerMagnitude (Ability|MakeOutgoingGameplayEffectSpec (Variables|Default|GetGE_Heals) 0.0) "(TagName=\"SetByCaller.Health\")" (Variables|Default|GetHealthHealAmount)))
      (bind _returnvalue_3 (Ability|GameplayEffect|AssignTagSetbyCallerMagnitude _returnvalue_2 "(TagName=\"SetByCaller.Shield\")" (Variables|Default|GetShieldHealAmount)))
      (Ability|ApplyGameplayEffectSpecToOwner _returnvalue_3)
      (Ability|EndAbility))))

(fn CalcMissingHealthShield
  (bind _playerpawn (Variables|Default|GetPlayerPawn))
  (Variables|Default|SetMissingHealth (select (Math|Float|NearlyEqual(Float) (FortPawn|GetMaxHealth _playerpawn) (FortPawn|GetHealth _playerpawn)) 0 (Math|Float|Floor (- (FortPawn|GetMaxHealth _playerpawn) (FortPawn|GetHealth _playerpawn)))))
  (Variables|Default|SetMissingShield (select (Math|Float|NearlyEqual(Float) (FortPawn|GetMaxShield _playerpawn) (FortPawn|GetShield _playerpawn)) 0 (Math|Float|Floor (- (FortPawn|GetMaxShield _playerpawn) (FortPawn|GetShield _playerpawn)))))
  (if (>= (Variables|Default|GetMissingHealth) (Math|Float|Ceil (Variables|Default|GetTotalHealAmount)))
    (Variables|Default|SetHealthHealAmount (Variables|Default|GetTotalHealAmount))
    (else
      (bind _output_get (Variables|Default|SetHealthHealAmount (Math|Conversions|ToFloat(Integer) (Variables|Default|GetMissingHealth))))
      (Variables|Default|SetShieldHealAmount (- (Variables|Default|GetTotalHealAmount) _output_get))))
  (return))
